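# Material function for vehicle cosmetics that provides a fake world environment reflection from a cubemap, with an overridable default.
# It samples the cubemap by reflection vector, both raw (LocalSpace output) and through a vector transform (WorldSpace output).
# Each sample is optionally desaturated, then blended toward a tint multiply by Tint Amount.
# The default tint comes from a curve atlas lookup, timed by a gradient texture read at a fixed UV.
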MaterialFunction /VehicleCosmetics/Mutable/FX_Materials/WorldEnv/MF_VC_WorldEnv
graph {nodes [
  n1 = ReflectionVectorWS()
  n2 = FunctionInput(Preview=TextureObject() {samplerType=SAMPLERTYPE_LinearColor, texture=&/WorldEnv_AthenaCubemap}) {bUsePreviewValueAsDefault=true, description=Use if you need custom cubemap, inputName=Cubemap, inputType=FunctionInput_TextureCube}
  n3 = FunctionInput(Preview=Constant()) {inputName=Desaturate, inputType=FunctionInput_Scalar}
  n4 = Desaturation(
   None=TextureSample(UVs=@n1, Tex=@n2) {samplerType=SAMPLERTYPE_LinearColor, texture=&WorldEnv_AthenaCubemap}.RGB
   Fraction=@n3
  )
  n5 = CurveAtlasRowParameter(CurveTime=TextureSample(UVs=Constant2Vector() {g=0.5, r=0.1667}) {texture=&/VC_WorldEnv_Gradient2_T}.R) {
   atlas=&/WordEnv_LookupAtlas
   curve=&/WorldEnv_ColorLookup
   parameterName=LookUpTexture
  }
  n6 = FunctionInput(Preview=@n5) {bUsePreviewValueAsDefault=true, inputName="Tint (Curve)"}
  n7 = FunctionInput(Preview=Constant() {r=1}) {bUsePreviewValueAsDefault=true, inputName=Tint Amount, inputType=FunctionInput_Scalar}
  FunctionOutput(None=LinearInterpolate(A=@n4, B=Multiply(A=@n4, B=@n6), Alpha=@n7)) {outputName=LocalSpace}
  n8 = Desaturation(
   None=TextureSample(UVs=Transform(None=@n1), Tex=@n2) {samplerType=SAMPLERTYPE_LinearColor, texture=&WorldEnv_AthenaCubemap}.RGB
   Fraction=@n3
  )
  FunctionOutput(None=LinearInterpolate(A=@n8, B=Multiply(A=@n8, B=@n6), Alpha=@n7)) {outputName=WorldSpace}
 ]}
